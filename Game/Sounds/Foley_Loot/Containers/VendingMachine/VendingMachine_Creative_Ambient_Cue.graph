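# Sound cue for the ambient hum of a Creative vending machine, routed to the Creative gadgets sound class.
# A single looping wave player plays the hum wave at reduced volume, with no randomization or mixing.
# Spatialization and falloff come from a shared mono item ambient-loop attenuation asset, as the cue's stored attenuation overrides are not switched on.
# Its own concurrency override caps how many hums play at once.
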SoundCue /Game/Sounds/Foley_Loot/Containers/VendingMachine/VendingMachine_Creative_Ambient_Cue
root=WavePlayer_2
nodes [{id=WavePlayer_2, properties {bLooping=true, soundWaveAssetPtr=&/VendingMachine_Hum_Loop_06}}]
properties {
 attenuationSettings=/Game/Sounds/Attenuation/Fort_Item_Ambient_Loop_Mono_Default_Att
 bOverrideConcurrency=true
 maxDistance=1250
 soundClassObject=&Fort_Gadgets_Creative
 volumeMultiplier=0.3
 attenuationOverrides {
  bEnableFocusInterpolation=true
  bEnableListenerFocus=true
  falloffDistance=600
  focusAttackInterpSpeed=0.5
  focusReleaseInterpSpeed=0.5
  lPFFrequencyAtMax=10000
  lPFRadiusMax=1000
  lPFRadiusMin=500
  nonFocusAzimuth=45
  nonFocusVolumeAttenuation=0.8
  nonSpatializedRadiusStart=200
  occlusionInterpolationTime=0.3
  occlusionLowPassFilterFrequency=3000
  occlusionVolumeAttenuation=0.75
  stereoSpread=1000
  attenuationShapeExtents {x=0}
  customAttenuationCurve {editorCurveData {keys [<at 0 1 RCIM_Cubic>, {arriveTangent=-1, interpMode=RCIM_Cubic, leaveTangent=-1, time=0.5, value=0.1}, <at 1 0 RCIM_Cubic>]}}
 }
 concurrencyOverrides {maxCount=4, volumeScaleReleaseTime=1}
}
