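# Sound cue for the vending machine vend sound in Creative, filed under loot container foley.
# A single wave player feeds the output directly with VendingMachine_Vend_02, with no randomization or modulation.
# It uses the shared item-interact attenuation and routes through the Fort_Loot_Reward_PSM sound class.
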SoundCue /Game/Sounds/Foley_Loot/Containers/VendingMachine/VendingMachine_Creative_Vend_Cue
root=WavePlayer_0
nodes [{id=WavePlayer_0, properties {soundWaveAssetPtr=&/VendingMachine_Vend_02}}]
properties {
 attenuationSettings=/Game/Sounds/Attenuation/Fort_Item_Interact_Att
 maxDistance=1850
 soundClassObject=&Fort_Loot_Reward_PSM
 volumeMultiplier=1
}
