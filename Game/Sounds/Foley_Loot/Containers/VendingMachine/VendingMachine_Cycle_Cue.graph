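# Sound cue for the vending machine cycle sound, filed under loot-container foley and assigned to the Fort_Loot_Collect sound class.
# A single wave player feeds the output directly with VendingMachine_Cycle_03, with no randomization and a slightly reduced volume.
# Overridden attenuation takes the place of the referenced prop-search asset, adding short-range falloff and interpolated listener focus that boosts the sound in focus and quiets it outside.
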SoundCue /Game/Sounds/Foley_Loot/Containers/VendingMachine/VendingMachine_Cycle_Cue
root=WavePlayer_0
nodes [{id=WavePlayer_0, properties {soundWaveAssetPtr=&/VendingMachine_Cycle_03}}]
properties {
 attenuationSettings=/Game/Sounds/Attenuation/Fort_Prop_Search_Att
 bOverrideAttenuation=true
 maxDistance=1000
 soundClassObject=&Fort_Loot_Collect
 volumeMultiplier=0.796
 attenuationOverrides {
  bEnableFocusInterpolation=true
  bEnableListenerFocus=true
  falloffDistance=1000
  focusAttackInterpSpeed=20
  focusAzimuth=20
  focusReleaseInterpSpeed=10
  focusVolumeAttenuation=1.6
  nonFocusAzimuth=40
  nonFocusVolumeAttenuation=0.6
  nonSpatializedRadiusStart=200
  attenuationShapeExtents {x=0}
 }
}
