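# Sound cue for the end of a vending machine search, apparently signalling that no loot is available.
# A mixer plays the vending machine fail wave and the shared UI "no more loot" wave together at equal volume.
# It uses the mono item-interact attenuation and plays through the loot collect sound class.
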SoundCue /Game/Sounds/Foley_Loot/Containers/VendingMachine/VendingMachine_SearchEnd_Cue
root=Mixer_0
nodes [
 {id=Mixer_0, children [WavePlayer_0, WavePlayer_14], properties {inputVolume [1,1]}}
 {id=WavePlayer_0, properties {soundWaveAssetPtr=&/VendingMachine_Fail}}
 {id=WavePlayer_14, properties {soundWaveAssetPtr=/Game/Sounds/Fort_UI_Sounds/Gameplay/Fort_UI_NoMoreLoot}}
]
properties {
 attenuationSettings=/Game/Sounds/Attenuation/Fort_Item_Interact_Mono_Att
 maxDistance=1500
 soundClassObject=&Fort_Loot_Collect
 volumeMultiplier=1
}
