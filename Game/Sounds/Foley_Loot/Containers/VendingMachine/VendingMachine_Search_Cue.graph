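# Sound cue for the vending machine search foley, in the loot container sound set.
# A single wave player feeds the VendingMachine_Search wave straight to the output, with no randomization or mixing.
# It routes through the Fort_Loot_Collect sound class and uses the shared mono item-interact attenuation at slightly reduced volume.
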SoundCue /Game/Sounds/Foley_Loot/Containers/VendingMachine/VendingMachine_Search_Cue
root=WavePlayer_0
nodes [{id=WavePlayer_0, properties {soundWaveAssetPtr=&/VendingMachine_Search}}]
properties {
 attenuationSettings=/Game/Sounds/Attenuation/Fort_Item_Interact_Mono_Att
 maxDistance=1500
 soundClassObject=&Fort_Loot_Collect
 volumeMultiplier=0.7
}
