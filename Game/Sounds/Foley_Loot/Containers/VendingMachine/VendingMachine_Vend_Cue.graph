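# Sound cue for a vending machine vend, part of the loot container foley.
# A mixer plays the main vend wave immediately alongside a sparkle layer that starts after a fixed half-second delay, with the main wave at full volume and the sparkle slightly lower.
# It uses the item-interact attenuation and routes to the loot reward sound class.
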SoundCue /Game/Sounds/Foley_Loot/Containers/VendingMachine/VendingMachine_Vend_Cue
root=Mixer_0
nodes [
 {id=Delay_0, children [WavePlayer_1], properties {delayMax=0.5, delayMin=0.5}}
 {id=Mixer_0, children [Delay_0,WavePlayer_0], properties {inputVolume [0.7,1]}}
 {id=WavePlayer_0, properties {soundWaveAssetPtr=&/VendingMachine_Vend_02}}
 {id=WavePlayer_1, properties {soundWaveAssetPtr=&/VendingMachine_Vend_Sparkle}}
]
properties {
 attenuationSettings=/Game/Sounds/Attenuation/Fort_Item_Interact_Att
 maxDistance=1850
 soundClassObject=&Fort_Loot_Reward_PSM
 volumeMultiplier=1
}
